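# Physics asset for the skeletal mesh of the MilitaryFashion pickaxe, a melee weapon.
# A single capsule body on the Melee_TwoHanded_Handle bone provides the collision shape, offset and rotated relative to that bone.
# It has no constraints, so the mesh is represented as one body with no jointed parts.
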PhysicsAsset /Game/Weapons/FORT_Melee/Pickaxe_MilitaryFashion/Meshes/SK_Pickaxe_MilitaryFashion_PhysicsAsset
bodies {Melee_TwoHanded_Handle [{shapeType=Capsule, length=92.7378, radius=29.0205, center=<0.9483 1.9079 26.4493>, rotation=<rot 1.1563 -90 -0.0825>}]}
constraints []
